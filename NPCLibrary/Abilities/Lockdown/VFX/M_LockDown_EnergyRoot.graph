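# Unlit translucent Niagara mesh-particle material for the Lockdown ability's energy root effect, drawing animated radial energy bands.
# Emissive blends two colours through a soft radial mask, tinted by particle colour.
# Opacity pans a tiled radial coordinate with a per-particle time offset, distorts it with polar noise and shapes it through a curve atlas borrowed from AwfulDuck.
# A soft ring mask, depth fade and particle alpha limit it; explicit UV derivatives drive noise mip selection.
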Material /NPCLibrary/Abilities/Lockdown/VFX/M_LockDown_EnergyRoot
graph {
 properties {
  bUsedWithNiagaraMeshParticles=true
  blendMode=BLEND_Translucent
  pixelDepthOffsetMode=PDOM_Legacy
  shadingModel=MSM_Unlit
  usedShadingModels=Unlit
 }
 nodes [
  n1 = VectorParameter() {parameterName=ColorA, defaultValue=<rgba 0.2047 0 3 0>}
  n2 = VectorParameter() {parameterName=ColorB, defaultValue=<rgba 0.8604 0.8428 1.5 0>}
  n3 = Constant() {r=0.5}
  n4 = SphereMask(A=TextureCoordinate(), B=@n3, Radius=@n3) {hardnessPercent=0}
  n5 = ParticleColor()
  MaterialFunctionCall(Emissive=Multiply(A=LinearInterpolate(A=@n1.RGB, B=@n2.RGB, Alpha=@n4), B=@n5.RGB)) {materialFunction=&MF_FX_EmissiveAdjustments}
  n6 = TextureCoordinate()
  n7 = Constant() {r=0.5}
  n8 = Subtract(
   A=Power(Base=SphereMask(A=@n6, B=@n7, Radius=@n7) {hardnessPercent=0})
   B=SphereMask(A=TextureCoordinate(), B=Constant() {r=0.5}, Radius=Constant() {r=0.18}) {hardnessPercent=0}
  )
  n9 = ScalarParameter() {defaultValue=1}
  n10 = ScalarParameter() {defaultValue=1, parameterName=Param_3}
  n11 = TextureCoordinate()
  n12 = TextureSample!(
   UVs=Multiply(
    A=MaterialFunctionCall() {materialFunction=&VectorToRadialValue}.Radial Coordinates
    B=AppendVector(A=@n9, B=@n10)
   )
   Tex
   "DDX(UVs)"=DDX(Value=@n11)
   "DDY(UVs)"=DDY(Value=@n11)
   Apply View MipBias
  ) {mipValueMode=TMVM_Derivative, samplerType=SAMPLERTYPE_Grayscale, texture=&LowResBlurredNoise}
  n13 = ScalarParameter() {defaultValue=0.5, parameterName=Param_2}
  n14 = MaterialFunctionCall() {materialFunction=&VectorToRadialValue}
  n15 = Time()
  n16 = ScalarParameter() {defaultValue=1.2, parameterName=PanSpeed}
  n17 = Add(
   A=Multiply(
    A=MaterialFunctionCall(Float2=@n14.Radial Coordinates) {materialFunction=&BreakOutFloat2Components}.G
    B=ScalarParameter() {defaultValue=1.8, parameterName=Tiling, sortPriority=0}
   ) {constB=3}
   B=Add(A=Multiply(A=@n15, B=@n16), B=DynamicParameter() {paramNames [RandomTimeOffset, Param2, Param3, Param4]}.RandomTimeOffset)
  )
  n18 = CurveAtlasRowParameter(CurveTime=Add(A=Multiply(A=@n12.R, B=@n13), B=Frac(None=@n17))) {
   atlas=/Game/Athena/Items/AwfulDuck/FX/CA_AwfulDuck
   curve=/Game/Athena/Items/AwfulDuck/FX/Curve_ExpCurve
   defaultValue=10
   parameterName=Param_1
   sortPriority=0
  }
  n19 = Multiply(A=@n18.A)
  Multiply(A=@n5.A, B=DepthFade(Opacity=Multiply(A=@n8, B=@n19)) {fadeDistanceDefault=20})
 ]
}
