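# Unlit translucent energy-link beam material for the NPC Lockdown ability VFX, built for Niagara ribbons and mesh particles.
# Three noise-distorted panning streak samples form offset RGB channels, lighten-blended and tinted, with a Grow-driven sine pulse and LinkedFlash boosting brightness.
# Opacity combines curve-atlas UV falloffs, the pulse and flash, and depth fade.
# Time-based sines wobble vertices, while UseSpline blends to alternate UVs and SplineThicken offsets with reduced wobble.
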Material /NPCLibrary/Abilities/Lockdown/VFX/M_Lockdonw_EnergyLink
graph {
 properties {
  bUsedWithNiagaraMeshParticles=true
  bUsedWithNiagaraRibbons=true
  blendMode=BLEND_Translucent
  pixelDepthOffsetMode=PDOM_Legacy
  shadingModel=MSM_Unlit
  twoSided=true
  usedShadingModels=Unlit
 }
 nodes [
  LinearInterpolate()
  MaterialFunctionCall() {materialFunction=&DebugTimeSine}
  MaterialFunctionCall() {materialFunction=&DebugTimeSine}
  n1 = NamedRerouteUsage()
  n2 = LinearInterpolate(Alpha=DynamicParameter() {parameterIndex=1, defaultValue=<rgba 1 0 1 1>, paramNames [Grow,LinkedFlash,Param3,Param4]}.Grow) {constA=-1}
  n3 = Sine(None=Subtract(
    A=MaterialFunctionCall(Float2=@n1) {materialFunction=&BreakOutFloat2Components}.R
    B=Subtract(A=@n2) {constB=0.2}
   )) {period=2}
  n4 = Multiply(A=Saturate(None=ConstantBiasScale(None=@n3) {bias=-0.95})) {constB=25}
  n5 = VectorParameter() {parameterName=Tiling, defaultValue=<rgba 4 1 3 1>}
  n6 = Panner(Coordinate=Multiply(A=NamedRerouteUsage(), B=ComponentMask(None=@n5.RGB) {g=true, r=true})) {speedX=-1}
  n7 = TextureSample(UVs=Multiply(A=AppendVector(A=@n5.B, B=@n5.B), B=NamedRerouteUsage())) {samplerType=SAMPLERTYPE_Normal, texture=/Game/Characters/Player/Female/Medium/Bodies/F_MED_Bistro_Astronaut/FX/T_GaussianNoise_Normal}
  n8 = Divide(A=@n7.R) {constB=7}
  n9 = Multiply(
   A=TextureSample(UVs=Add(A=@n6, B=@n8)) {samplerType=SAMPLERTYPE_LinearColor, texture=/Game/TechArt/Seasonal/LackLuster/Textures/T_Lackluster_Streaks}.R
   B=Constant3Vector() {constant=<rgba 0 0 1 0>}
  )
  n10 = Divide(A=@n7.R) {constB=5}
  n11 = Multiply(
   A=Constant3Vector() {constant=<rgba 0 1 0 0>}
   B=TextureSample(UVs=Add(A=@n6, B=@n10)) {samplerType=SAMPLERTYPE_LinearColor, texture=&T_Lackluster_Streaks}.R
  )
  n12 = Constant3Vector() {constant=<rgba 1 0 0 0>}
  n13 = TextureSample(UVs=Add(A=@n6, B=Divide(A=@n7.R) {constB=3})) {samplerType=SAMPLERTYPE_LinearColor, texture=&T_Lackluster_Streaks}
  n14 = Multiply(
   A=MaterialFunctionCall!(
    Blend=MaterialFunctionCall(Blend=@n9, Base=@n11) {materialFunction=/Engine/Functions/Engine_MaterialFunctions03/Blends/Blend_Lighten}.Result
    Base=Multiply(A=@n12, B=@n13.R)
   ) {materialFunction=&Blend_Lighten}.Result
   B=VectorParameter() {parameterName=EmissiveColor, defaultValue=<rgba 0.2864 0.3303 1 0>}.RGB
  )
  n15 = Multiply(
   A=Add(A=Multiply(A=@n4, B=@n14), B=@n14)
   B=LinearInterpolate(Alpha=DynamicParameter() {parameterIndex=1, defaultValue=<rgba 1 0 1 1>, paramNames [Grow,LinkedFlash,Param3,Param4]}.LinkedFlash) {constA=1, constB=4}
  )
  MaterialFunctionCall(Emissive=@n15) {materialFunction=&MF_FX_EmissiveAdjustments}
  NamedRerouteDeclaration(None=LinearInterpolate(A=TextureCoordinate(), B=TextureCoordinate() {coordinateIndex=1}, Alpha=ScalarParameter() {parameterName=UseSpline})) {name=TexCoord, nodeColor=<rgba 1 0 0.0474 1>}
  NamedRerouteDeclaration(None=@n15) {name=Emissive, nodeColor=<rgba 1 0 0.1875 1>}
  n16 = MaterialFunctionCall(Float2=NamedRerouteUsage()) {materialFunction=&BreakOutFloat2Components}
  n17 = CurveAtlasRowParameter(CurveTime=@n16.R) {atlas=/Game/Athena/Items/AwfulDuck/FX/CA_AwfulDuck, curve=/Game/Athena/Items/AwfulDuck/FX/Curve_ExpCurve, defaultValue=10}
  n18 = CurveAtlasRowParameter(CurveTime=@n16.G) {atlas=&CA_AwfulDuck, curve=&Curve_ExpCurve, defaultValue=10}
  n19 = Multiply(A=LinearInterpolate(A=@n17.G, B=@n17.G, Alpha=ScalarParameter() {parameterName=UseSpline}), B=@n18.B)
  n20 = MaterialFunctionCall(Float2=NamedRerouteUsage()) {materialFunction=&BreakOutFloat2Components}
  n21 = Multiply(A=OneMinus(None=@n20.R)) {constB=6}
  n22 = Multiply(A=Saturate(None=@n21), B=@n18.B)
  n23 = Subtract(
   A=MaterialFunctionCall(Float2=NamedRerouteUsage()) {materialFunction=&BreakOutFloat2Components}.R
   B=LinearInterpolate(Alpha=DynamicParameter() {parameterIndex=1, defaultValue=<rgba 1 0 1 1>, paramNames [Grow,LinkedFlash,Param3,Param4]}.Grow) {constA=-1}
  )
  n24 = Saturate(None=ConstantBiasScale(None=Sine(None=@n23) {period=2}) {bias=-0.6})
  n25 = Add(
   A=Saturate(None=Multiply(A=@n24) {constB=5})
   B=DynamicParameter() {parameterIndex=1, defaultValue=<rgba 1 0 1 1>, paramNames [Grow,LinkedFlash,Param3,Param4]}.LinkedFlash
  )
  n26 = DepthFade(
   Opacity=Multiply(A=LinearInterpolate(A=@n19, B=@n22, Alpha=@n25), B=ScalarParameter() {defaultValue=1, parameterName=Opacity})
   FadeDistance=ScalarParameter() {defaultValue=100, parameterName=DepthFade}
  )
  NamedRerouteDeclaration(None=ShadingPathSwitch(Default=@n26, Deferred=@n26)) {name=Opacity, nodeColor=<rgba 0.2891 0 1 1>}
  n27 = NamedRerouteUsage()
  n28 = NamedRerouteUsage()
  n29 = Divide(A=Add(A=Time()) {constB=0.33}) {constB=3}
  n30 = DynamicParameter() {defaultValue=<rgba 1 0 1 1>, paramNames [EmitterRandom, Grow, Param3, Param4]}
  n31 = Sine(None=Add(A=Divide(A=@n28), B=Add(A=@n29, B=@n30.EmitterRandom)))
  n32 = ScalarParameter() {defaultValue=100, parameterName=WPO_Wave}
  n33 = NamedRerouteUsage()
  n34 = Divide(A=Time()) {constB=3}
  n35 = DynamicParameter() {defaultValue=<rgba 1 0 1 1>, paramNames [EmitterRandom, Grow, Param3, Param4]}
  n36 = Sine(None=Add(A=Divide(A=@n33), B=Add(A=@n34, B=@n35.EmitterRandom)))
  n37 = NamedRerouteUsage()
  n38 = Divide(A=Add(A=Time()) {constB=-0.33}) {constB=3}
  n39 = DynamicParameter() {defaultValue=<rgba 1 0 1 1>, paramNames [EmitterRandom, Grow, Param3, Param4]}
  n40 = Sine(None=Add(A=Divide(A=@n37), B=Add(A=@n38, B=@n39.EmitterRandom)))
  n41 = Multiply(
   A=CurveAtlasRowParameter(CurveTime=MaterialFunctionCall(Float2=@n27) {materialFunction=&BreakOutFloat2Components}.R) {atlas=&CA_AwfulDuck, curve=&Curve_ExpCurve, defaultValue=10}.B
   B=MaterialFunctionCall(R=Multiply(A=@n31, B=@n32), G=Multiply(A=@n36, B=@n32), B=Multiply(A=@n40, B=@n32)) {materialFunction=&AppendMany}.RGB
  )
  n42 = ScalarParameter() {defaultValue=40, parameterName=SplineWidth}
  n43 = MaterialFunctionCall(WidthBase=@n42, WidthTip=@n42) {materialFunction=&SplineThicken}
  n44 = Divide(A=@n41) {constB=8}
  NamedRerouteDeclaration(None=LinearInterpolate(A=@n41, B=Add(A=@n43.WorldPositionOffset, B=@n44), Alpha=ScalarParameter() {parameterName=UseSpline})) {name=WPO, nodeColor=<rgba 0.8047 1 0 1>}
  NamedRerouteUsage()
  NamedRerouteUsage()
  NamedRerouteUsage()
  Sine(None=Multiply(A=@n20.R) {constB=0.5})
  Subtract(A=@n21) {constB=0.2}
 ]
}
